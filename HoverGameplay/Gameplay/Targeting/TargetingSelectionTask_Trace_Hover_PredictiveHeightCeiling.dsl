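; Targeting trace selection task for hover gameplay that aims a predictive height-ceiling trace along the source actor's motion.
; GetTraceVector scales the actor's horizontal velocity by a velocity multiplier and subtracts the height limit on Z, both read from scalable floats.
; GetTraceDirection and GetTraceLength derive the normalized direction and magnitude from that vector using the targeting request's source context.
; An invalid subsystem or actor yields zero.
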(fn GetTraceVector
  (params (SourceActor :type Actor%20Object%20Reference))
  (Utilities|IsValid SourceActor
    (:Is%20Valid
      (return (Utilities|Operators|Subtract (* (Math|Conversions|ToVector2D(Vector) (Transformation|GetVelocity SourceActor)) (ScalableFloat|GetValueatLevel (Variables|Default|GetVelocityMultiplier))) (Math|Vector|MakeVector 0.0 0.0 (ScalableFloat|GetValueatLevel (Variables|Default|GetHeightLimit))))))
    (:Is%20Not%20Valid
      (return "0, 0, 0"))))

(fn GetTraceDirection
  (params (TargetingHandle :type Targeting%20Request%20Handle%20Structure%20%28by%20ref%29))
  (bind _returnvalue (Targeting|GetTargetingSubsystem TargetingHandle))
  (Utilities|IsValid _returnvalue
    (:Is%20Valid
      (return (Math|Vector|Normalize (CallFunction|GetTraceVector (Utilities|Struct|BreakTargetingSourceContext (TargetingSystem|TargetingTypes|GetTargetingSourceContext _returnvalue TargetingHandle))))))
    (:Is%20Not%20Valid
      (return "0, 0, 0"))))

(fn GetTraceLength
  (params (TargetingHandle :type Targeting%20Request%20Handle%20Structure%20%28by%20ref%29))
  (bind _returnvalue (Targeting|GetTargetingSubsystem TargetingHandle))
  (Utilities|IsValid _returnvalue
    (:Is%20Valid
      (return (Math|Vector|VectorLength (CallFunction|GetTraceVector (Utilities|Struct|BreakTargetingSourceContext (TargetingSystem|TargetingTypes|GetTargetingSourceContext _returnvalue TargetingHandle))))))
    (:Is%20Not%20Valid
      (return 0.0))))
